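# Additive unlit Niagara sprite material for fragment effects on the Motorcyclist Female pickaxe.
# Particle color RGB runs through the shared MF_FX_EmissiveAdjustments function.
# The red channel of a 2x2 fragments texture is eroded by subtracting inverted, saturated particle alpha, so shapes shrink away as alpha drops.
# A Frac of Time node is present but feeds nothing shown.
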Material /BRCosmetics/Weapons/FORT_Melee/Pickaxe_Motorcyclist_Female/FX/Materials/M_Pickaxe_Motorcyclist_Female_Fragments
graph {
 properties {bUsedWithNiagaraSprites=true, blendMode=BLEND_Additive, pixelDepthOffsetMode=PDOM_Legacy, shadingModel=MSM_Unlit, usedShadingModels=Unlit}
 nodes [
  Frac(None=Time())
  n1 = ParticleColor()
  MaterialFunctionCall(Emissive=@n1.RGB) {materialFunction=&MF_FX_EmissiveAdjustments}
  n2 = TextureSample(UVs=TextureCoordinate()) {
   samplerType=SAMPLERTYPE_LinearGrayscale
   texture=/BRCosmetics/Weapons/FORT_Melee/Pickaxe_Motorcyclist_Female/FX/Textures/T_Pickaxe_Motorcyclist_Female_Fragments_2x2
  }
  Subtract(A=@n2.R, B=OneMinus(None=Saturate(None=@n1.A)))
 ]
}
