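# Unlit translucent Niagara sprite material for the spike flash on Motorcyclist Female pickaxe impacts.
# Particle color starts overbright and fades to black over particle relative time, while mobile uses plain particle color.
# A scalar blends in a tiled gradient texture tint before MF_FX_EmissiveAdjustments produces the emissive.
# Opacity comes from unscaled spike flash texture alpha times particle alpha.
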Material /BRCosmetics/Weapons/FORT_Melee/Pickaxe_Motorcyclist_Female/FX/Materials/M_Pickaxe_Motorcyclist_Female_Impact_Spike
graph {
 properties {bUsedWithNiagaraSprites=true, blendMode=BLEND_Translucent, pixelDepthOffsetMode=PDOM_Legacy, shadingModel=MSM_Unlit, usedShadingModels=Unlit}
 nodes [
  n1 = ParticleColor()
  n2 = ShadingPathSwitch(
   Default=LinearInterpolate(A=Multiply(B=@n1.RGB) {constA=8}, B=Multiply(A=@n1.RGB) {constB=0}, Alpha=ParticleRelativeTime()) {constA=6}
   Mobile=@n1.RGB
  ) {desc=Migrated from FeatureLevelSwitch}
  n3 = ScalarParameter() {defaultValue=1, parameterName=CustomTexUV_X}
  n4 = ScalarParameter() {defaultValue=1, parameterName=CustomTexUV_Y}
  n5 = TextureSampleParameter2D(UVs=Multiply(A=TextureCoordinate(), B=AppendVector(A=@n3, B=@n4))) {parameterName=TextureGradient, texture=/Game/Effects/Fort_Effects/Textures/Explo/T_Flash_Spike_01}
  MaterialFunctionCall(Emissive=LinearInterpolate(A=@n2, B=Multiply(A=@n5.RGB, B=@n2), Alpha=ScalarParameter() {parameterName=UseCustomTex?})) {materialFunction=&MF_FX_EmissiveAdjustments}
  Multiply(A=TextureSample() {texture=&T_Flash_Spike_01}.A, B=ParticleColor().A)
 ]
}
