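# Niagara swing trail effect for the Motorcyclist Female pickaxe, with four selectable emissive color variants.
# An anim-trail ribbon with time-of-day alpha scaling joins a burst glow mesh with light and socket-spawned, drag-slowed sprites.
# User.EmissiveColorSelection indexes the glow color array and drives nested rounded lerps that pick among four color curves for ribbon and sprites.
# The Empty emitter has no spawn module.
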NiagaraSystem /BRCosmetics/Weapons/FORT_Melee/Pickaxe_Motorcyclist_Female/FX/NS_Pickaxe_Motorcyclist_Female_Trail
dependencies {
 usedDataInterfaces [&NiagaraDataInterfaceColorCurve, &NiagaraDataInterfaceCurve, &NiagaraDataInterfaceArrayColor]
 usedDynamicInputs []
 usedModules [&SystemState]
}
emitters {
 Empty {
  emitterUpdateScript [&EmitterState()]
  particleSpawnScript [&InitializeParticle(Color=<rgba 0 74 111 1>, Lifetime=5, UsePositionOffset=false, Write Material Random=true)]
  particleUpdateScript [&ParticleState()]
  rendererClasses [&NiagaraSpriteRendererProperties]
  renderers [{rendererClass=&NiagaraSpriteRendererProperties}]
 }
 FN_AnimTrail {
  emitterUpdateScript [
   &EmitterState(Life Cycle Mode=::NewEnumerator1, Loop Behavior=::NewEnumerator1, Loop Delay=0.05, Loop Duration=0.3, UseLoopDelay=true)
   &FN_SpawnAnimTrail_v2()
  ]
  particleSpawnScript [
   /Game/Effects/Niagara/Modules/AnimTrail/FN_InitializeAnimTrail/FN_InitializeAnimTrailBoneIndecies(First Bone or Socket Index=1)
   /Game/Effects/Niagara/Modules/AnimTrail/FN_InitializeAnimTrail/FN_InitializeAnimTrail_v3/FN_InitializeAnimTrail(Color:LinearColor=$User.TrailColor, Lifetime=0.1333, Write Material Random=false)
  ]
  particleUpdateScript [
   /Game/Effects/Niagara/Modules/AnimTrail/FN_UpdateAnimTrail/FN_UpdateAnimTrail_v3/FN_UpdateAnimTrail(Locked Particle RibbonWidth:NiagaraFloat=$Particles.RibbonWidth)
   /SetVariables(
    Particles.TrailColorCurve1:LinearColor=&ColorFromCurve(ColorCurve=DataInterface() {propertyValues {
       AlphaCurve {keys [<at 0 1>,<at 1 1>]}
       BlueCurve {keys [<at 0.28 9>, {arriveTangent=-0.0001, leaveTangent=-0.0001, time=0.286, value=9}]}
       GreenCurve {keys [{arriveTangent=-1720.738, time=0.28, value=9}, {arriveTangent=-0.4891, leaveTangent=-0.4891, time=0.286, value=0}]}
       RedCurve {keys [{arriveTangent=30.0741, time=0.28, value=0}, {arriveTangent=52.7832, leaveTangent=52.7832, time=0.286, value=9}]}
      }})
    Particles.TrailColorCurve2:LinearColor=&ColorFromCurve(ColorCurve=DataInterface() {propertyValues {
       AlphaCurve {keys [<at 0 1>,<at 1 1>]}
       BlueCurve {keys [<at 0.28 2.25>, {arriveTangent=-0.0001, leaveTangent=-0.0001, time=0.286, value=9}]}
       GreenCurve {keys [{arriveTangent=-1720.738, time=0.28, value=0}, {arriveTangent=-0.4891, leaveTangent=-0.4891, time=0.286, value=1.8}]}
       RedCurve {keys [{arriveTangent=30.0741, time=0.28, value=9}, {arriveTangent=52.7832, leaveTangent=52.7832, time=0.286, value=9}]}
      }})
    Particles.TrailColorCurve3:LinearColor=&ColorFromCurve(ColorCurve=DataInterface() {propertyValues {
       AlphaCurve {keys [<at 0 1>,<at 1 1>]}
       BlueCurve {keys [<at 0.283 0>, {arriveTangent=-0.0001, leaveTangent=-0.0001, time=0.289, value=9}]}
       GreenCurve {keys [{arriveTangent=-1720.738, time=0.283, value=5.7}, {arriveTangent=-0.4891, leaveTangent=-0.4891, time=0.289, value=0}]}
       RedCurve {keys [{arriveTangent=30.0741, time=0.283, value=9}, {arriveTangent=52.7832, leaveTangent=52.7832, time=0.289, value=9}]}
      }})
    Particles.TrailColorCurve4:LinearColor=&ColorFromCurve(ColorCurve=DataInterface() {propertyValues {
       AlphaCurve {keys [<at 0 1>,<at 1 1>]}
       BlueCurve {keys [<at 0.28 9>, {arriveTangent=-0.0001, leaveTangent=-0.0001, time=0.286, value=9}]}
       GreenCurve {keys [{arriveTangent=-1720.738, time=0.28, value=9}, {arriveTangent=-0.4891, leaveTangent=-0.4891, time=0.286, value=1.35}]}
       RedCurve {keys [{arriveTangent=30.0741, time=0.28, value=8.1}, {arriveTangent=52.7832, leaveTangent=52.7832, time=0.286, value=9}]}
      }})
   ) {isSetParameters=true}
   &Color(Color=&LerpLinearColors(
     StartColor=&LerpLinearColors(
      StartColor=$Particles.TrailColorCurve1
      EndColor=$Particles.TrailColorCurve2
      LerpFactor=/Niagara/DynamicInputs/Integer/V2/Round_Float(Float:NiagaraFloat=&ClampFloat(Float=&Subtract_Float(A=$User.EmissiveColorSelection, B=0)))
     )
     EndColor=&LerpLinearColors(
      StartColor=$Particles.TrailColorCurve3
      EndColor=$Particles.TrailColorCurve4
      LerpFactor=&Round_Float(Float:NiagaraFloat=&ClampFloat(Float=&Subtract_Float(A=$User.EmissiveColorSelection, B=2)))
     )
     LerpFactor=&Round_Float(Float:NiagaraFloat=&ClampFloat(Float=&Subtract_Float(A=$User.EmissiveColorSelection, B=1)))
    ))
   &ScaleColor(
    Color Value To Scale=$Particles.Color
    Scale Alpha=&Multiply_Float(
     A=&FloatFromCurve(
      FloatCurve=DataInterface() {propertyValues {Curve {keys [{arriveTangent=19.9617, interpMode=RCIM_Cubic, tangentMode=RCTM_User, time=0, value=1}, <at 1 0>]}}}
      Scale Curve=0.7
     )
     B=&FNTimeOfDayScalar(Night Scale=0.3, Morning Scale=0.5, Evening Scale=0.5)
    )
    ScaleA=true
    ScaleRGB=true
   )
  ]
  rendererClasses [&NiagaraRibbonRendererProperties]
  renderers [{rendererClass=&NiagaraRibbonRendererProperties}]
 }
 Mesh_Glow {
  emitterUpdateScript [
   &EmitterState(Life Cycle Mode=::NewEnumerator1, Loop Behavior=::NewEnumerator1, Loop Delay=0.05, Loop Duration=0.3, UseLoopDelay=true)
   &SpawnBurst_Instantaneous(Use Spawn Probability=false, UseLoopCountLimit=false)
  ]
  particleSpawnScript [&InitializeParticle(
    Color=&SelectLinearColorFromArray(
     Array Sampling Mode=::NewEnumerator1
     Direct Array Index=&MakeIntFromFloat(Float=$User.EmissiveColorSelection)
     Color Selection Array=DataInterface() {propertyValues {ColorData [<rgba 0 74 111 1>, <rgba 111 0 77.7 1>, <rgba 111 83.25 0 1>, <rgba 111 111 111 1>]}}
     Random Seed=?
    )
    Lifetime=0.3
    UsePositionOffset=false
    Write Material Random=true
   )]
  particleUpdateScript [
   &ParticleState()
   &ScaleColor(
    Scale Alpha=&FloatFromCurve(FloatCurve=DataInterface() {propertyValues {Curve {keys [
         {interpMode=RCIM_Cubic, tangentMode=RCTM_User, time=0, value=0}
         {interpMode=RCIM_Cubic, tangentMode=RCTM_User, time=0.5, value=1}
         {tangentMode=RCTM_User, time=1, value=0}
        ]}}})
    ScaleA=true
    ScaleRGB=true
   )
   /SetVariables(
    Particles.Light Color:LinearColor=&MultiplyLinearColor(Linear Color=$Particles.Color, Scale Factor=<1 1 1 0.03>)
    Particles.Light Position:Vector3f=&Add_Vector(
     A=Linked() {linkedVariable {name=Particles.Position, type=NiagaraPosition}}
     B=/Niagara/DynamicInputs/Transforms/TransformPosition(Position=<2 7 75>, Source Space=::Local, Destination Space=::Simulation)
    )
   ) {isSetParameters=true}
  ]
  rendererClasses [&NiagaraMeshRendererProperties, &NiagaraLightRendererProperties]
  renderers [
   {rendererClass=&NiagaraMeshRendererProperties, rendererIndex=0}
   {rendererClass=&NiagaraLightRendererProperties, rendererIndex=1}
  ]
 }
 Particle_Trail {
  emitterUpdateScript [
   &EmitterState(Life Cycle Mode=::NewEnumerator1, Loop Behavior=::NewEnumerator1, Loop Duration=0.3, UseLoopDelay=false)
   &SpawnRate(SpawnRate=99, Use Spawn Probability=false)
  ]
  particleSpawnScript [
   &InitializeParticle(
    Color Mode=::NewEnumerator0
    Lifetime=0.222
    Sprite Rotation Mode=::NewEnumerator2
    Sprite Size Mode=::NewEnumerator1
    Uniform Sprite Size Max=4
    Uniform Sprite Size Min=2
    UsePositionOffset=false
    Write Material Random=true
   )
   &InheritVelocity(Speed Limit=true, UseSpeedThreshold=false)
   &FN_BoneFlow(
    BoneID:NiagaraInt32=&RandomRangeInt(Minimum=0, Maximum=0)
    Height:NiagaraFloat=&RandomRangeFloat(Minimum=0, Maximum=0.5)
    SkeletalMesh:NiagaraDataInterfaceSkeletalMesh=$User.SkeletalMesh
    Use Sockets Instead=true
   )
   /SetVariables(Particles.SubImageIndex=&RandomRangeFloat(Minimum=0, Maximum=4)) {isSetParameters=true}
  ]
  particleUpdateScript [
   &ParticleState()
   &Drag(Drag=5, Use Linear Drag=true, Use Rotational Drag=false)
   &SolveForcesAndVelocity(Clamp Velocity=false, Limit Acceleration=false)
   /SetVariables(
    Particles.ParticleColorCurve1:LinearColor=&ColorFromCurve(ColorCurve=DataInterface() {propertyValues {
       AlphaCurve {keys [<at 0.3132 1 RCIM_Cubic>,<at 1 0 RCIM_Cubic>]}
       BlueCurve {keys [<at 0.2036 2 RCIM_Cubic>,<at 0.3445 2>,<at 0.5237 2 RCIM_Cubic>]}
       GreenCurve {keys [{arriveTangent=-14.1913, interpMode=RCIM_Cubic, leaveTangent=-14.1913, time=0.2036, value=2}, <at 0.3445 0>, <at 0.5237 0 RCIM_Cubic>]}
       RedCurve {keys [<at 0.2036 0 RCIM_Cubic>, <at 0.3445 0>, {arriveTangent=11.1602, interpMode=RCIM_Cubic, leaveTangent=11.1602, time=0.5237, value=2}]}
      }})
    Particles.ParticleColorCurve2:LinearColor=&ColorFromCurve(ColorCurve=DataInterface() {propertyValues {
       AlphaCurve {keys [<at 0.3132 1 RCIM_Cubic>,<at 1 0 RCIM_Cubic>]}
       BlueCurve {keys [
         {arriveTangent=5.44, interpMode=RCIM_Cubic, leaveTangent=5.44, time=0.2036, value=0.5}
         <at 0.3445 1.2667>
         {arriveTangent=4.0921, interpMode=RCIM_Cubic, leaveTangent=4.0921, time=0.5237, value=2}
        ]}
       GreenCurve {keys [<at 0.2036 0 RCIM_Cubic>,<at 0.3445 0>,<at 0.5237 0 RCIM_Cubic>]}
       RedCurve {keys [<at 0.2036 2 RCIM_Cubic>,<at 0.3445 2>,<at 0.5237 2 RCIM_Cubic>]}
      }})
    Particles.ParticleColorCurve3:LinearColor=&ColorFromCurve(ColorCurve=DataInterface() {propertyValues {
       AlphaCurve {keys [<at 0.3132 1 RCIM_Cubic>,<at 1 0 RCIM_Cubic>]}
       BlueCurve {keys [
         {arriveTangent=0.2365, interpMode=RCIM_Cubic, leaveTangent=0.2365, time=0.2036, value=0}
         <at 0.3445 0.0333>
         {arriveTangent=10.9742, interpMode=RCIM_Cubic, leaveTangent=10.9742, time=0.5237, value=2}
        ]}
       GreenCurve {keys [{arriveTangent=-8.9879, interpMode=RCIM_Cubic, leaveTangent=-8.9879, time=0.2036, value=1.2667}, <at 0.3445 0>, <at 0.5237 0 RCIM_Cubic>]}
       RedCurve {keys [<at 0.2036 2 RCIM_Cubic>,<at 0.3445 2>,<at 0.5237 2 RCIM_Cubic>]}
      }})
    Particles.ParticleColorCurve4:LinearColor=&ColorFromCurve(ColorCurve=DataInterface() {propertyValues {
       AlphaCurve {keys [<at 0.3132 1 RCIM_Cubic>,<at 1 0 RCIM_Cubic>]}
       BlueCurve {keys [<at 0.2036 2 RCIM_Cubic>,<at 0.3445 2>,<at 0.5237 2 RCIM_Cubic>]}
       GreenCurve {keys [
         {arriveTangent=-10.6435, interpMode=RCIM_Cubic, leaveTangent=-10.6435, time=0.2036, value=2}
         <at 0.3445 0.5>
         {arriveTangent=-2.7901, interpMode=RCIM_Cubic, leaveTangent=-2.7901, time=0.5237, value=0}
        ]}
       RedCurve {keys [<at 0.2036 2 RCIM_Cubic>,<at 0.3445 2>,<at 0.5237 2 RCIM_Cubic>]}
      }})
   ) {isSetParameters=true}
   &Color(Color=&LerpLinearColors(
     StartColor=&LerpLinearColors(
      StartColor=$Particles.ParticleColorCurve1
      EndColor=$Particles.ParticleColorCurve2
      LerpFactor=&Round_Float(Float:NiagaraFloat=&ClampFloat(Float=&Subtract_Float(A=$User.EmissiveColorSelection, B=0)))
     )
     EndColor=&LerpLinearColors(
      StartColor=$Particles.ParticleColorCurve3
      EndColor=$Particles.ParticleColorCurve4
      LerpFactor=&Round_Float(Float:NiagaraFloat=&ClampFloat(Float=&Subtract_Float(A=$User.EmissiveColorSelection, B=2)))
     )
     LerpFactor=&Round_Float(Float:NiagaraFloat=&ClampFloat(Float=&Subtract_Float(A=$User.EmissiveColorSelection, B=1)))
    ))
  ]
  rendererClasses [&NiagaraSpriteRendererProperties]
  renderers [{rendererClass=&NiagaraSpriteRendererProperties}]
 }
}
summary {
 systemName=NS_Pickaxe_Motorcyclist_Female_Trail
 emitters [
  {emitterName=FN_AnimTrail, rendererClasses [&NiagaraRibbonRendererProperties]}
  {emitterName=Mesh_Glow, rendererClasses [&NiagaraMeshRendererProperties, &NiagaraLightRendererProperties]}
  {emitterName=Particle_Trail, rendererClasses [&NiagaraSpriteRendererProperties]}
  {emitterName=Empty, rendererClasses [&NiagaraSpriteRendererProperties]}
 ]
 userVariables [
  User.EmissiveColorSelection {type=NiagaraFloat, defaultValue=0, description=""}
  User.RibbonIDOffset {type=NiagaraInt32}
  User.SkeletalMesh {
   type=NiagaraDataInterfaceSkeletalMesh
   defaultValue {
    struct=NiagaraExt_VariableValue_DataInterface
    value {dataInterface=&self:NiagaraDataInterfaceSkeletalMesh_0}
   }
  }
  User.TrailColor {type=LinearColor, description="", defaultValue {struct=LinearColor, value=<rgba 1 1 1 0.35>}}
 ]
}
